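# XP_Min-Challenges_Post_Cue is a Sound Cue in the Athena mini-challenge quest sounds, named for an XP post sting.
# The file shows no nodes and no root, so no wave is connected.
# It carries only settings: the Fort_Music_POI_PSM sound class, reduced volume, a very large max distance and a concurrency override that allows one instance and blocks new ones.
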SoundCue /Game/Athena/Sounds/Quests/MiniChallenges/XP_Min-Challenges_Post_Cue
root=null
nodes []
properties {
 bOverrideConcurrency=true
 maxDistance=2097152
 soundClassObject=/Game/Sounds/FortSoundClassesAndModes/Fort_Music_POI_PSM
 volumeMultiplier=0.7
 concurrencyOverrides {maxCount=1, resolutionRule=PreventNew}
}
