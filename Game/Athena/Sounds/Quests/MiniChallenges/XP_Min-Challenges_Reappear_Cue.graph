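# Sound cue for the reappear sound of mini challenges XP in Athena quests, classed as non-critical in-game UI audio.
# A fixed delay node holds back a single wave player that plays the Season 13 reappear wave.
# A concurrency override allows one instance at a time, stopping the oldest with a short voice-steal release, and the cue volume is slightly boosted.
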SoundCue /Game/Athena/Sounds/Quests/MiniChallenges/XP_Min-Challenges_Reappear_Cue
root=Delay_0
nodes [
 {id=Delay_0, children [WavePlayer_0], properties {delayMax=1.3, delayMin=1.3}}
 {id=WavePlayer_0, properties {soundWaveAssetPtr=&/S13_XP_Min-Challenges_Reappear}}
]
properties {
 bOverrideConcurrency=true
 maxDistance=2097152
 soundClassObject=&Fort_UI_InGame_NonCritical
 volumeMultiplier=1.2
 concurrencyOverrides {maxCount=1, resolutionRule=StopOldest, voiceStealReleaseTime=0.25}
}
